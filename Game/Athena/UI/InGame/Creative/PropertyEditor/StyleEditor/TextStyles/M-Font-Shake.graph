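# Translucent unlit UI font material that, by its name and path, gives text a shake style in the Creative property editor.
# Time-driven sine waves with separate X and Y speeds are remapped and scaled by Bounce Width and Bounce Height into a small UV offset.
# A Color parameter supplies the tint through a U-masked lerp whose endpoints are identical, so the color stays uniform.
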Material /Game/Athena/UI/InGame/Creative/PropertyEditor/StyleEditor/TextStyles/M-Font-Shake
graph {
 properties {
  bUsedWithStaticMesh=false
  blendMode=BLEND_Translucent
  materialDomain=MD_UI
  numCustomizedUVs=1
  pixelDepthOffsetMode=PDOM_Legacy
  shadingModel=MSM_Unlit
  usedShadingModels=Unlit
 }
 nodes [
  n1 = Sine(None=Multiply(A=Time(), B=ScalarParameter() {defaultValue=10, parameterName=Speed Y, sliderMax=100, sliderMin=-10}))
  n2 = LinearInterpolate(A=Constant() {r=-0.1}, B=Constant() {r=0.1}, Alpha=Multiply(A=Add(A=@n1)) {constB=0.5})
  n3 = Multiply(A=ScalarParameter() {defaultValue=5, parameterName=Bounce Height, sliderMax=5, sliderMin=-5}) {constB=0.001}
  n4 = Sine(None=Multiply(A=Time(), B=ScalarParameter() {defaultValue=2, parameterName=Speed X}))
  n5 = LinearInterpolate(A=Constant() {r=-0.1}, B=Constant() {r=0.1}, Alpha=Multiply(A=Add(A=@n4)) {constB=0.5})
  n6 = Multiply(A=ScalarParameter() {parameterName=Bounce Width, sliderMax=5, sliderMin=-5}) {constB=0.001}
  Add(A=TextureCoordinate(), B=AppendVector(A=Multiply(A=@n5, B=@n6) {constB=0.05}, B=Multiply(A=@n2, B=@n3) {constB=0.05}))
  n7 = VectorParameter() {parameterName=Color, defaultValue=<rgba 1 1 1 1>}
  n8 = ComponentMask(None=TextureCoordinate()) {r=true}
  Reroute(None=LinearInterpolate(A=@n7.RGB, B=@n7.RGB, Alpha=Ceil(None=@n8)))
 ]
}
